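# Translucent parent material for the Poach vehicle wheel cosmetic effect, drawing an animated glowing burst on skeletal meshes.
# A static light-burst texture sample distorts the UVs of a second sample of that texture, which either pans or rotates via static switch.
# The channel-masked, clamped result drives opacity, Paint_Color-tinted emissive and an overlay base color.
# Attributes then pass through vehicle damage and de-rez transition functions.
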Material /VehicleCosmetics/Wheels/Poach/Materials/M_Parent_Poach_Effect_01
graph {
 properties {
  bAutomaticallySetUsageInEditor=false
  bUseMaterialAttributes=true
  bUsedWithSkeletalMesh=true
  blendMode=BLEND_Translucent
  usedShadingModels=Default Lit
 }
 nodes [
  n1 = TextureSample(UVs=Add(A=TextureCoordinate() {uTiling=0.7, vTiling=0.7}, B=Constant2Vector() {g=0.15, r=0.15})) {samplerType=SAMPLERTYPE_LinearColor, texture=/VehicleCosmetics/Wheels/_SharedTextures/FX/LightBurstAlpha_RGBA}
  n2 = Add(A=@n1.B, B=TextureCoordinate())
  n3 = TextureCoordinate() {uTiling=2, vTiling=2}
  n4 = Multiply(A=ScalarParameter() {defaultValue=3, parameterName=Time_Multyplier_on UV_Distortion}, B=Time())
  n5 = Add(
   A=Add(A=Add(A=@n1.A, B=@n2), B=@n1.R)
   B=StaticSwitchParameter(True=Panner(Coordinate=@n3, Time=@n4) {speedX=0.2, speedY=0.2}, False=Rotator(Coordinate=@n3, Time=@n4) {centerX=1, centerY=1, speed=0.3}) {parameterName=T-PAN___F_Rotate}
  )
  n6 = Multiply(
   A=Multiply(
    A=TextureSample(UVs=@n5) {samplerType=SAMPLERTYPE_LinearColor, texture=&LightBurstAlpha_RGBA}.A
    B=Constant() {r=0.2}
   )
   B=ScalarParameter() {defaultValue=3.4, parameterName=MovingTexturePresenceMultiplier}
  )
  n7 = Multiply(A=Add(A=@n6, B=Floor(None=@n6)), B=@n1.R)
  n8 = ScalarParameter() {defaultValue=1, parameterName=Color_Streight}
  n9 = Multiply(A=@n7, B=Multiply(A=@n8, B=@n1.R))
  n10 = Multiply(A=@n8, B=@n1.G)
  n11 = Add(A=@n7, B=Multiply(A=Multiply(A=@n9, B=@n10), B=Multiply(A=@n8, B=@n1.B)))
  n12 = Power(Base=Constant() {r=1.1}, Exp=Constant() {r=1.25})
  n13 = Clamp(None=Multiply(A=Multiply(A=@n11, B=@n12), B=@n11)) {maxDefault=0.8}
  n14 = Multiply(A=VectorParameter() {parameterName=Extra_Color_Overlay_Diffuse, defaultValue=<rgba 0 0 0 1>}.RGB, B=@n13)
  n15 = VectorParameter() {parameterName=Paint_Color, defaultValue=<rgba 0 2 0.0853 1>}
  n16 = Multiply(A=Multiply(A=@n13, B=ScalarParameter() {defaultValue=2, parameterName=Emissive_Multyplier}), B=@n15.RGB)
  MaterialFunctionCall(In=MaterialFunctionCall(Material=SetMaterialAttributes!(MaterialAttributes, Base Color=@n14, Emissive Color=@n16, Opacity=@n13) {attributeSetTypes [#1, #2, #3]}) {materialFunction=&MF_BaseVehicle_Damage}.Result) {materialFunction=&MF_VehicleDeRez_Transition}
  Multiply(A=@n15.RGB, B=@n1.R)
  Multiply(A=@n15.RGB, B=@n1.G)
  Multiply(A=@n15.RGB, B=@n1.B)
 ]
}
