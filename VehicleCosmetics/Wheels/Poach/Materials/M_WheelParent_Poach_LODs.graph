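# Masked parent material for the LODs of the Poach vehicle wheel, usable on skeletal meshes.
# It takes material attributes from MF_WheelParent_Simple and overrides only Opacity Mask, blending toward a rim alpha texture where vertex color red is low.
# The result passes through MF_BaseVehicle_Damage and then MF_VehicleDeRez_Transition, layering in shared vehicle damage and de-rez transition functions.
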Material /VehicleCosmetics/Wheels/Poach/Materials/M_WheelParent_Poach_LODs
graph {
 properties {
  bAutomaticallySetUsageInEditor=false
  bUseMaterialAttributes=true
  bUsedWithMeshDeformer=true
  bUsedWithSkeletalMesh=true
  blendMode=BLEND_Masked
  pixelDepthOffsetMode=PDOM_Legacy
  usedShadingModels=Default Lit
 }
 nodes [
  n1 = GetMaterialAttributes(None=MaterialFunctionCall() {materialFunction=&MF_WheelParent_Simple}.MaterialAttributes) {attributeGetTypes [#1]}
  n2 = TextureSampleParameter2D() {parameterName=Alpha_Mask_LODs, samplerType=SAMPLERTYPE_LinearColor, texture=/VehicleCosmetics/Wheels/Poach/Textures/T_Poach_Rim_Alpha_LOD}
  n3 = LinearInterpolate(A=@n1.Opacity Mask, B=@n2.R, Alpha=OneMinus(None=VertexColor().R))
  MaterialFunctionCall(In=MaterialFunctionCall(Material=SetMaterialAttributes!(MaterialAttributes=@n1.MaterialAttributes, Opacity Mask=@n3) {attributeSetTypes [#1]}) {materialFunction=&MF_BaseVehicle_Damage}.Result) {materialFunction=&MF_VehicleDeRez_Transition}
 ]
}
